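; Vehicle cosmetics component that applies a looping "rez" FX GameplayCue locally on its owning Fortnite vehicle.
; On BeginPlay it binds cue add and remove to cosmetics FX start and finish delegates, skipping NoRezFX-tagged vehicles and DM-mode vehicles outside Creative.
; Removal is marked temporary until swap removal can be delayed.
; Dev-only events spawn collisionless boost and trail test actors attached to mesh sockets, and destroy them.
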(event Custom|RemoveInitialGC
  ;[1 1 1] Temp setup untill we can delay removal of swap
  (Projectile|RemoveGameplayCue%28Looping%29Local (Variables|Default|GetVehiclePawn) (Variables|Default|GetGC_LoopingApplication) (Ability|GameplayCue|MakeGameplayCueParameters :NormalizedMagnitude 0.0 :RawMagnitude 0.0 :Location "0,0,0" :Normal "0,0,0" :Instigator (Variables|Default|GetVehiclePawn) :GameplayEffectLevel 1 :AbilityLevel 1 :bReplicateLocationWhenUsingMinimalRepProxy false)))

(event Custom|ApplyInitialGC
  ;[1 1 1] Initial GC application
  (Projectile|AddGameplayCue%28Looping%29Local (Variables|Default|GetVehiclePawn) (Variables|Default|GetGC_LoopingApplication) (Ability|GameplayCue|MakeGameplayCueParameters :NormalizedMagnitude 0.0 :RawMagnitude 0.0 :Location "0,0,0" :Normal "0,0,0" :Instigator (Variables|Default|GetVehiclePawn) :GameplayEffectLevel 1 :AbilityLevel 1 :bReplicateLocationWhenUsingMinimalRepProxy false)))

(event EventBeginPlay
  (bind _returnvalue (Components|GetOwner))
  (Utilities|IsValid _returnvalue
    (:Is%20Valid
      (bind _asfort_athena_vehicle (Utilities|Casting|CastToFortAthenaVehicle _returnvalue)
        (:then
          (bind _output_get (Variables|Default|SetVehiclePawn _asfort_athena_vehicle))
          (if (and (Utilities|IsValid _output_get :!class /Script/Engine.KismetSystemLibrary) (Utilities|IsValid (Fortnite|Athena|GameState|GetGameStateAthena) :!class /Script/Engine.KismetSystemLibrary))
            (if (not (or (GameplayTags|HasTag (Class|FortVehicleInterface|GetVehicleTags _output_get) "(TagName=\"Vehicle.Cosmetics.NoRezFX\")") (and (GameplayTags|HasTag (Class|FortVehicleInterface|GetVehicleTags _output_get) "(TagName=\"Vehicle.Modetype.DM\")") (not (Class|FortGameStateAthena|IsCreativeModePlaylist (Fortnite|Athena|GameState|GetGameStateAthena))))))
              (Vehicle|Cosmetics|FX|BindEventtoVehicleCosmeticsStartFXDelegate (EventDispatchers|CreateEvent :!event Custom|ApplyInitialGC))
              (Vehicle|Cosmetics|FX|BindEventtoVehicleCosmeticsFinishFXDelegate (EventDispatchers|CreateEvent :!event Custom|RemoveInitialGC)))))
        (:CastFailed
          _)))
    (:Is%20Not%20Valid
      _))
  (Development|PrintString (Utilities|String|Append "invalid var: " (Utilities|GetDisplayName _returnvalue))))

(event Custom|ValidateBoostTest
  ;[0 0 0 0.5] [Dev Only] Validation tests
  (CallFunction|EndValidateBoostTest)
  (bind _spawnedactor (CallFunction|SpawnActorforValidationTest (Variables|Default|GetBoostTestSocketName) (Variables|Default|GetBoostTestActorClass)))
  (Variables|Default|SetBoostTestSpawnedActor _spawnedactor))

(event Custom|EndValidateBoostTest
  ;[0 0 0 0.5] [Dev Only] Validation tests
  (Actor|DestroyActor (Variables|Default|GetBoostTestSpawnedActor)))

(event Custom|ValidateTrailTest
  ;[0 0 0 0.5] [Dev Only] Validation tests
  (CallFunction|EndValidateTrailTest)
  (bind _spawnedactor (CallFunction|SpawnActorforValidationTest (Variables|Default|GetTrailTestSocketName) (Variables|Default|GetTrailTestActorClass)))
  (Variables|Default|SetTrailTestSpawnedActor _spawnedactor))

(event Custom|EndValidateTrailTest
  ;[0 0 0 0.5] [Dev Only] Validation tests
  (Actor|DestroyActor (Variables|Default|GetTrailTestSpawnedActor)))

(fn SpawnActorForValidationTest
  (params (SocketName :type Name) (ActorClass :type Actor%20Class%20Reference))
  (bind _returnvalue (Components|GetOwner))
  (bind _returnvalue_2 (Class|FortVehicleInterface|GetSkeletalMeshComponent(Message) _returnvalue))
  (bind _returnvalue_3 (Game|SpawnActorfromClass ActorClass (Transformation|GetSocketTransform _returnvalue_2 SocketName "RTS_World") "AlwaysSpawn" "MultiplyWithRoot" _returnvalue 0))
  (bind _output_get (LocalVariables|SetLocalSpawnedActor _returnvalue_3))
  (Collision|SetActorEnableCollision _output_get false)
  (Transformation|AttachActorToActor _output_get _returnvalue SocketName "SnapToTarget" "SnapToTarget" "KeepWorld" false)
  (return _output_get))
